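# Looping idle cosmetic effect for the ScoopPluck player character: eye, swoosh and capsule glows plus crackling body and leg lightning.
# Three emitters use spawn groups to pin particles to skeleton sockets through a shared socket reader.
# Glows take Variant and time-of-day tints, while lightning pulses random meshes with glow sprites on looping spawn-rate curves.
# User flags for helmet, hidden head, hands or feet, and ADS hide or dim matching groups.
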NiagaraSystem /BRCosmetics/Characters/Player/Male/Medium/Bodies/M_MED_ScoopPluck/FX/NS_Character_ScoopPluck_Idle
dependencies {
 usedDataInterfaces [&NiagaraDataInterfaceCurve, &NiagaraDataInterfaceSocketReader, &NiagaraDataInterfaceColorCurve]
 usedDynamicInputs [&FNTimeOfDayScalar, &ColorFromCurve]
 usedModules [
  &self:SystemSpawnScript.SystemScriptSource.SystemScriptGraph.NiagaraNodeAssignment_1.SetVariables_#1
  &SystemState
  &self:SystemSpawnScript.SystemScriptSource.SystemScriptGraph.NiagaraNodeAssignment_3.SetVariables_#2
 ]
}
emitters {
 EyesSwooshesGlow {
  emitterUpdateScript [
   &EmitterState(Life Cycle Mode=::NewEnumerator1, Loop Behavior=::NewEnumerator1, Loop Duration Mode=::NewEnumerator1, UseLoopDelay=false)
   &SpawnBurst_Instantaneous(Spawn Count=7, Use Spawn Probability=false, UseLoopCountLimit=false)
  ]
  particleSpawnScript [
   &FN_SetSpawnGroup(Number of Spawn Groups=7)
   &InitializeParticle(
    Lifetime=3.7
    Position Mode=::NewEnumerator0
    Sprite Size=&Multiply_Vector2D_ByFloat(
     Float=11
     Vector2D=&MergeHelper_Vector2Switch(
      0=<1 1>
      1=<1 1>
      2=<1.5008 0.6>
      3=<1.5008 0.71>
      4=&Lerp_Vector2(A=<1.3923 2.5613>, B=<1.3923 2.5613>, Alpha=$User.Variant)
      5=<0.6364 1.6364>
      6=<0.6364 1.6364>
      Index=&SpawnGroup()
      Num=::NewEnumerator5
     )
    )
    Sprite Size Mode=::NewEnumerator4
    Write Material Random=true
   )
   &CameraOffset(Camera Offset Amount=&MergeHelper_FloatSwitch(0=3, 1=3, 2=1, 3=1, 4=3, 5=2, 6=2, Index=&SpawnGroup(), Num=::NewEnumerator5))
   /SetVariables(Particles.VisibilityTag=&MergeHelper_IntSwitch(0=0, 1=0, 2=1, 3=1, 4=2, 5=2, 6=2, Index=&SpawnGroup(), Num=::NewEnumerator5)) {isSetParameters=true}
  ]
  particleUpdateScript [
   &ParticleState()
   &SocketLocation(
    SamplePosOffset=true
    Sampled Position Offset=&MergeHelper_VectorSwitch(
     0=<3.6 10 2.4822>
     1=<3.6 10 -3.2992>
     2=<9.5 -1.3092 7.7888>
     3=<9.5 -1.3092 -7.7888>
     4=<11.6585 12 2.5033>
     5=<8.9772 -0.3497 7.7888>
     6=<8.9772 -0.3497 -7.7888>
     Index=&SpawnGroup()
     Num=::NewEnumerator5
    )
    Socket Index=&MergeHelper_IntSwitch(0=18, 1=18, 2=18, 3=18, 4=19, 5=18, 6=18, Index=&SpawnGroup(), Num=::NewEnumerator5)
    Socket Reader=$System.Socket Reader
    Socket Sampling Mode=::NewEnumerator3
    VelocitySpeedLimit=false
   )
   &SpriteFacingAndAlignment(
    Sprite Alignment=&RotateVectorByQuaternion(
     Vector To Rotate=&MergeHelper_VectorSwitch(
      0=<0 0 0>
      1=<0 0 0>
      2=<1 0.34 0>
      3=<1 0.34 0>
      4=<-0.232 0 1>
      5=<-0.8 1 -0.1128>
      6=<-0.8 1 0.1128>
      Index=&SpawnGroup()
      Num=::NewEnumerator5
     )
     Quaternion=$Output.SocketLocation.SampledRotation
    )
    Sprite Facing=&RotateVectorByQuaternion(
     Vector To Rotate=&MergeHelper_VectorSwitch(
      0=<0 0 0>
      1=<0 0 0>
      2=<0 -0.1 -1>
      3=<0 0.1 -1>
      4=<0 1 0>
      5=<1 0 0>
      6=<1 0 0>
      Index=&SpawnGroup()
      Num=::NewEnumerator5
     )
     Quaternion=$Output.SocketLocation.SampledRotation
    )
    WriteAlignment=true
    WriteFacing=true
   )
   &DynamicMaterialParameters(
    Index 0 Param 1=1
    Index 0 Param 3=5
    Module.Index0Param0Write=true
    Module.Index0Param1Write=false
    Module.Index0Param2Write=false
    Module.Index0Param3Write=false
   )
   &ScaleColor(
    Color Scale:LinearColor=&Multiply_LinearColorByFloat(
     Linear Color=&LerpLinearColors(StartColor=<rgba 0.2007 1.717 2 0.16>, EndColor=<rgba 2 0.0765 0.1491 0.22>, LerpFactor=$User.Variant)
     Float=&FN_FrontEndFloatModifier(Gameplay=1, Front End=0.7)
    )
    Scale Mode=3
   ) {module=ScaleColor004}
   &ScaleColor(
    Scale Alpha=&Multiply_Float(
     A=&FloatFromCurve(FloatCurve=DataInterface() {propertyValues {Curve {keys [
          <at 0 0 RCIM_Cubic>
          <at 0.1004 1 RCIM_Cubic>
          {arriveTangent=-3.3254, interpMode=RCIM_Cubic, leaveTangent=-3.3254, tangentMode=RCTM_User, time=1, value=0}
         ]}}})
     B=&MakeCustomFloatFromBool(
      Bool=&Boolean_OrOperation(Bool1=$User.bIsHeadHidden, Bool2=&SetBoolByFloatComparison(A=$User.Helmet, B=0.5))
      True Float=0
      False Float=1
     )
    )
    ScaleA=true
    ScaleRGB=false
   )
   &ScaleColor(Color Scale:LinearColor=$System.TOD Color, Color Value To Scale=$Particles.Color, Scale Mode=3) {module=ScaleColor001}
   &ScaleColor(Scale Alpha=0.4, ScaleA=true, ScaleRGB=false) {module=ScaleColor002}
   &ScaleColor(
    Scale Alpha=&MakeCustomFloatFromBool(Bool=$User.bIsADS, True Float=0.5, False Float=1)
    ScaleA=true
    ScaleRGB=false
   ) {module=ScaleColor003}
   /SetVariables(Particles.SwooshColor:LinearColor=&MakeLinearColorFromVectorAndFloat(
     "Vector (RGB)"=&MakeVectorFromLinearColorRGB(Color=&MultiplyLinearColor(
       Linear Color=<rgba 3 0.3125 0.4139 1>
       Scale Factor=&MakeVector4_FromLinearColor(LinearColor=$System.TOD Color)
      ))
     "Float (Alpha)"=&Multiply_Float(
      A=&Multiply_Float(
       A=&Multiply_Float(A=0.33, B=&FN_FrontEndFloatModifier(Gameplay=0.4, Front End=0.3))
       B=&Multiply_Float(
        A=&FloatFromCurve(FloatCurve=DataInterface() {propertyValues {Curve {keys [
             <at 0 0 RCIM_Cubic>
             <at 0.1004 1 RCIM_Cubic>
             {arriveTangent=-3.3254, interpMode=RCIM_Cubic, leaveTangent=-3.3254, tangentMode=RCTM_User, time=1, value=0}
            ]}}})
        B=&MakeCustomFloatFromBool(
         Bool=&Boolean_OrOperation(Bool1=$User.bIsHeadHidden, Bool2=&SetBoolByFloatComparison(A=$User.Helmet, B=0.5))
         True Float=0
         False Float=1
        )
       )
      )
      B=&MakeCustomFloatFromBool(Bool=$User.bIsADS, True Float=0.5, False Float=1)
     )
    )) {isSetParameters=true}
   /SetVariables(Particles.CapsuleGlowColor:LinearColor=&MakeLinearColorFromVectorAndFloat(
     "Vector (RGB)"=&MakeVectorFromLinearColorRGB(Color=<rgba 1 0.1042 0.138 1>)
     "Float (Alpha)"=&Multiply_Float(
      A=0.06
      B=&Multiply_Float(
       A=&FloatFromCurve(FloatCurve=DataInterface() {propertyValues {Curve {keys [
            <at 0 0 RCIM_Cubic>
            <at 0.1004 1 RCIM_Cubic>
            {arriveTangent=-3.3254, interpMode=RCIM_Cubic, leaveTangent=-3.3254, tangentMode=RCTM_User, time=1, value=0}
           ]}}})
       B=&MergeHelper_FloatSwitch(
        0=&MakeCustomFloatFromBool(
         Bool=&Boolean_OrOperation(Bool1=$User.bIsHeadHidden, Bool2=&SetBoolByFloatComparison(A=$User.Helmet, B=0.5))
         True Float=0
         False Float=1
        )
        1=&MakeCustomFloatFromBool(
         Bool=&Boolean_OrOperation(Bool1=$User.bIsHeadHidden, Bool2=&SetBoolByFloatComparison(A=$User.Helmet, B=0.5))
         True Float=0
         False Float=1
        )
        2=&MakeCustomFloatFromBool(
         Bool=&Boolean_OrOperation(Bool1=$User.bIsHeadHidden, Bool2=&SetBoolByFloatComparison(A=$User.Helmet, B=0.5))
         True Float=0
         False Float=1
        )
        3=&MakeCustomFloatFromBool(
         Bool=&Boolean_OrOperation(Bool1=$User.bIsHeadHidden, Bool2=&SetBoolByFloatComparison(A=$User.Helmet, B=0.5))
         True Float=0
         False Float=1
        )
        4=1
        5=&MakeCustomFloatFromBool(
         Bool=&Boolean_OrOperation(Bool1=$User.bIsHeadHidden, Bool2=&SetBoolByFloatComparison(A=$User.Helmet, B=0.5))
         True Float=0
         False Float=1
        )
        6=&MakeCustomFloatFromBool(
         Bool=&Boolean_OrOperation(Bool1=$User.bIsHeadHidden, Bool2=&SetBoolByFloatComparison(A=$User.Helmet, B=0.5))
         True Float=0
         False Float=1
        )
        Index=&SpawnGroup()
        Num=::NewEnumerator5
       )
      )
     )
    )) {isSetParameters=true}
  ]
  rendererClasses [&NiagaraSpriteRendererProperties]
  renderers [
   {rendererClass=&NiagaraSpriteRendererProperties}
   {rendererClass=&NiagaraSpriteRendererProperties, rendererIndex=1}
   {rendererClass=&NiagaraSpriteRendererProperties, rendererIndex=2}
  ]
 }
 LightningBody {
  emitterUpdateScript [
   &EmitterState(Inactive Response=::NewEnumerator1, Life Cycle Mode=::NewEnumerator1, Loop Behavior=::NewEnumerator1, Loop Duration=3.25, UseLoopDelay=false)
   &SpawnRate(
    SpawnRate=&FloatFromCurve(
     FloatCurve=DataInterface() {propertyValues {Curve {keys [
         <at 0 0 RCIM_Cubic>
         {arriveTangent=-0.8372, interpMode=RCIM_Cubic, leaveTangent=-0.8372, tangentMode=RCTM_User, time=0.1004, value=1.5}
         {arriveTangent=-0.4848, interpMode=RCIM_Cubic, leaveTangent=-0.4848, tangentMode=RCTM_User, time=0.5164, value=0.4626}
         <at 1 0 RCIM_Cubic>
        ]}}}
     CurveIndex=$Emitter.NormalizedLoopAge
     Scale Curve=19
    )
    Use Spawn Probability=false
   )
  ]
  particleSpawnScript [
   &FN_SetSpawnGroup(Number of Spawn Groups=10)
   &InitializeParticle(
    Color=<rgba 1 0.878 0.3092 1>
    Lifetime=&RandomRangeFloat(Minimum=0.4, Maximum=0.55)
    Mesh Scale=&MergeHelper_VectorSwitch(
     0=&RandomRangeVector(Minimum=<0.1 0.2 0.1>, Maximum=<0.2 0.5 0.2>)
     1=&RandomRangeVector(Minimum=<0.1 0.2 0.1>, Maximum=<0.2 0.5 0.2>)
     2=&RandomRangeVector(Minimum=<0.2 0.5 0.2>, Maximum=<0.5 0.8 0.5>)
     3=&RandomRangeVector(Minimum=<0.2 0.5 0.2>, Maximum=<0.5 0.8 0.5>)
     4=&RandomRangeVector(Minimum=<0.1 0.4 0.2>, Maximum=<0.3 0.7 0.4>)
     5=&RandomRangeVector(Minimum=<0.1 0.4 0.2>, Maximum=<0.3 0.7 0.4>)
     6=&RandomRangeVector(Minimum=<0.2 0.4 0.2>, Maximum=<0.4 0.7 0.4>)
     7=&RandomRangeVector(Minimum=<0.2 0.4 0.2>, Maximum=<0.4 0.7 0.4>)
     8=&RandomRangeVector(Minimum=<0.1 0.4 0.1>, Maximum=<0.3 0.5 0.3>)
     9=&RandomRangeVector(Minimum=<0.1 0.4 0.1>, Maximum=<0.3 0.5 0.3>)
     Index=&SpawnGroup()
     Num=::NewEnumerator8
    )
    Mesh Scale Mode=::NewEnumerator4
    UsePositionOffset=false
    Write Material Random=true
   )
   &InitialMeshOrientation(
    Mesh Orientation Mode=::NewEnumerator2
    Orientation Vector=<0 1 0>
    Rotation=&MergeHelper_VectorSwitch(
     0=&RandomRangeVector(Minimum=<0 0 0>, Maximum=<0 1 0>)
     1=&RandomRangeVector(Minimum=<0 0 0>, Maximum=<0 1 0>)
     2=&RandomRangeVector(Minimum=<0 0 0>, Maximum=<0 1 0>)
     3=&RandomRangeVector(Minimum=<0 0 0>, Maximum=<0 1 0>)
     4=<0 0 0>
     5=<0 -0.032 0>
     6=<-0.032 0 0>
     7=<0 0 0>
     8=<0 0 0>
     9=<0 0 0>
     Index=&SpawnGroup()
     Num=::NewEnumerator8
    )
    Use Rotation Vector=true
   )
   &CameraOffset(Camera Offset Amount=14)
  ]
  particleUpdateScript [
   &ParticleState()
   &ScaleMeshSize(Scale Factor=&VectorFromFloat(Value=&MergeHelper_FloatSwitch(
      0=&MakeCustomFloatFromBool(Bool=&Boolean_OrOperation(Bool1=$User.bIsHandHidden, Bool2=$User.bIsADS), True Float=0, False Float=1)
      1=&MakeCustomFloatFromBool(Bool=&Boolean_OrOperation(Bool1=$User.bIsHandHidden, Bool2=$User.bIsADS), True Float=0, False Float=1)
      2=1
      3=1
      4=1
      5=1
      6=1
      7=1
      8=1
      9=1
      Index=&SpawnGroup()
      Num=::NewEnumerator8
     )))
   &ScaleColor(
    Scale Alpha=&MergeHelper_FloatSwitch(
     0=&MakeCustomFloatFromBool(Bool=&Boolean_OrOperation(Bool1=$User.bIsHandHidden, Bool2=$User.bIsADS), True Float=0, False Float=1)
     1=&MakeCustomFloatFromBool(Bool=&Boolean_OrOperation(Bool1=$User.bIsHandHidden, Bool2=$User.bIsADS), True Float=0, False Float=1)
     2=1
     3=1
     4=1
     5=1
     6=1
     7=1
     8=1
     9=1
     Index=&SpawnGroup()
     Num=::NewEnumerator8
    )
    ScaleA=true
    ScaleRGB=true
   )
   &DynamicMaterialParameters(
    Index 0 Param 1=&FloatFromCurve(FloatCurve=DataInterface() {propertyValues {Curve {keys [{tangentMode=RCTM_User, time=0, value=0}, <at 0.333 0.5>, {tangentMode=RCTM_User, time=1, value=1}]}}})
    Index 0 Param 2=&FloatFromCurve(
     FloatCurve=DataInterface() {propertyValues {Curve {keys [<at 0.333 1>,<at 0.666 0.333>]}}}
     Scale Curve=&RandomRangeFloat(Minimum=0.6, Maximum=1)
    )
    Index 0 Param 3=0.75
    Index 0 Param 4=&FloatFromCurve(
     FloatCurve=DataInterface() {propertyValues {Curve {keys [
         {interpMode=RCIM_Cubic, tangentMode=RCTM_User, time=0, value=0}
         {interpMode=RCIM_Cubic, tangentMode=RCTM_User, time=0.5, value=1}
         {tangentMode=RCTM_User, time=1, value=0}
        ]}}}
     Scale Curve=33
    )
    Index 1 Param 1=1
    Module.Index0Param0Write=true
    Module.Index0Param1Write=true
    Module.Index0Param2Write=true
    Module.Index0Param3Write=true
    Module.Index1Param0Write=true
    Module.Index1Param1Write=false
    Module.Index1Param2Write=false
    Module.Index1Param3Write=false
    Write Parameter Index 1=true
   )
   /SetVariables(Particles.MeshIndex=&RandomRangeInt(Minimum=0, Maximum=4)) {isSetParameters=true, module=SetVariables_#3}
   &SocketLocation(
    SamplePosOffset=true
    Sampled Position Offset=&MergeHelper_VectorSwitch(
     0=&RandomRangeVector(Minimum=<12 0 0>, Maximum=<18 0 0>)
     1=&RandomRangeVector(Minimum=<12 0 0>, Maximum=<18 0 0>)
     2=&RandomRangeVector(Minimum=<17 0 0>, Maximum=<5 0 0>)
     3=&RandomRangeVector(Minimum=<17 0 0>, Maximum=<5 0 0>)
     4=&RandomRangeVector(Minimum=<-2 0 -5>, Maximum=<-7 0 5>)
     5=&RandomRangeVector(Minimum=<-2 -3 -5>, Maximum=<-7 -3 5>)
     6=&RandomRangeVector(Minimum=<-5 0 0>, Maximum=<-5 0 10>)
     7=&RandomRangeVector(Minimum=<-5 0 -4>, Maximum=<-5 0 5>)
     8=&RandomRangeVector(Minimum=<0 -4 0>, Maximum=<0 3 0>)
     9=&RandomRangeVector(Minimum=<0 -4 -6>, Maximum=<0 3 0>)
     Index=&SpawnGroup()
     Num=::NewEnumerator8
    )
    Socket Index=&SpawnGroup()
    Socket Reader=$System.Socket Reader
    Socket Sampling Mode=::NewEnumerator3
    VelocitySpeedLimit=false
   )
   &FN_Sprite_Attributes(
    Camera Offset=15
    Color=&MakeLinearColorFromVectorAndFloat(
     "Vector (RGB)"=&Multiply_Vector(A=&MakeVectorFromLinearColorRGB(Color=<rgba 1 0.6535 0.0208 1>), B=<1 1 1>)
     "Float (Alpha)"=&Multiply_Float(
      A=&FloatFromCurve(
       FloatCurve=DataInterface() {propertyValues {Curve {keys [
           {interpMode=RCIM_Cubic, tangentMode=RCTM_User, time=0, value=0}
           {interpMode=RCIM_Cubic, tangentMode=RCTM_User, time=0.3, value=1}
           {tangentMode=RCTM_User, time=1, value=0}
          ]}}}
       Scale Curve=0.7
      )
      B=1
     )
    )
    Create Camera Offset Attribute=true
    Create Color Attribute=true
    Create SpriteSize Attribute=true
    SpriteSize=&Multiply_Vector2D_ByFloat(
     Float=&FloatFromCurve(
      FloatCurve=DataInterface() {propertyValues {Curve {keys [
          {interpMode=RCIM_Cubic, tangentMode=RCTM_User, time=0, value=0}
          {interpMode=RCIM_Cubic, tangentMode=RCTM_User, time=0.3, value=1}
          {tangentMode=RCTM_User, time=1, value=0}
         ]}}}
      Scale Curve=&MergeHelper_FloatSwitch(
       0=&MakeCustomFloatFromBool(Bool=$User.bIsHandHidden, True Float=0, False Float=1)
       1=&MakeCustomFloatFromBool(Bool=$User.bIsHandHidden, True Float=0, False Float=1)
       2=1
       3=1
       4=1
       5=1
       6=1
       7=1
       8=1
       9=1
       Index=&SpawnGroup()
       Num=::NewEnumerator8
      )
     )
     Vector2D=<25 25>
    )
    Write to Module Namespace=true
   )
  ]
  rendererClasses [&NiagaraMeshRendererProperties, &NiagaraSpriteRendererProperties]
  renderers [
   {rendererClass=&NiagaraMeshRendererProperties, rendererIndex=0}
   {rendererClass=&NiagaraMeshRendererProperties, rendererIndex=1}
   {rendererClass=&NiagaraMeshRendererProperties, rendererIndex=2}
   {rendererClass=&NiagaraMeshRendererProperties, rendererIndex=3}
   {rendererClass=&NiagaraSpriteRendererProperties, rendererIndex=4}
  ]
 }
 LightningLegs {
  emitterUpdateScript [
   &EmitterState(Inactive Response=::NewEnumerator1, Life Cycle Mode=::NewEnumerator1, Loop Behavior=::NewEnumerator1, Loop Duration=3.1, UseLoopDelay=false)
   &SpawnRate(
    SpawnRate=&FloatFromCurve(
     FloatCurve=DataInterface() {propertyValues {Curve {keys [
         <at 0 0 RCIM_Cubic>
         {arriveTangent=-0.8372, interpMode=RCIM_Cubic, leaveTangent=-0.8372, tangentMode=RCTM_User, time=0.1004, value=1.5}
         {arriveTangent=-0.4848, interpMode=RCIM_Cubic, leaveTangent=-0.4848, tangentMode=RCTM_User, time=0.5164, value=0.4626}
         <at 1 0 RCIM_Cubic>
        ]}}}
     CurveIndex=$Emitter.NormalizedLoopAge
     Scale Curve=17
    )
    Use Spawn Probability=false
   )
  ]
  particleSpawnScript [
   &FN_SetSpawnGroup(Number of Spawn Groups=8)
   &InitializeParticle(
    Color=<rgba 1 0.878 0.3092 1>
    Lifetime=&RandomRangeFloat(Minimum=0.4, Maximum=0.55)
    Mesh Scale=&MergeHelper_VectorSwitch(
     0=&RandomRangeVector(Minimum=<0.25 0.5 0.25>, Maximum=<0.4 0.8 0.4>)
     1=&RandomRangeVector(Minimum=<0.25 0.5 0.25>, Maximum=<0.4 0.8 0.4>)
     2=&RandomRangeVector(Minimum=<0.2 0.5 0.2>, Maximum=<0.35 0.8 0.35>)
     3=&RandomRangeVector(Minimum=<0.2 0.5 0.2>, Maximum=<0.35 0.8 0.35>)
     4=&RandomRangeVector(Minimum=<0.1 0.5 0.1>, Maximum=<0.2 0.9 0.2>)
     5=&RandomRangeVector(Minimum=<0.1 0.5 0.1>, Maximum=<0.2 0.9 0.2>)
     6=&RandomRangeVector(Minimum=<0.1 0.5 0.1>, Maximum=<0.2 0.9 0.2>)
     7=&RandomRangeVector(Minimum=<0.1 0.5 0.1>, Maximum=<0.2 0.9 0.2>)
     Index=&SpawnGroup()
     Num=::NewEnumerator6
    )
    Mesh Scale Mode=::NewEnumerator4
    UsePositionOffset=false
    Write Material Random=true
   )
   &InitialMeshOrientation(
    Mesh Orientation Mode=::NewEnumerator2
    Orientation Vector=<0 1 0>
    Rotation=&MergeHelper_VectorSwitch(
     0=&RandomRangeVector(Minimum=<0 0 0>, Maximum=<0 1 0>)
     1=&RandomRangeVector(Minimum=<0 0 0>, Maximum=<0 1 0>)
     2=&RandomRangeVector(Minimum=<0 0 0>, Maximum=<0 1 0>)
     3=&RandomRangeVector(Minimum=<0 0 0>, Maximum=<0 1 0>)
     4=&RandomRangeVector(Minimum=<0 -0.3 0>, Maximum=<0 0.1 0>)
     5=&RandomRangeVector(Minimum=<0 0.55 0>, Maximum=<0 0.9 0>)
     6=&RandomRangeVector(Minimum=<0 0.3 0>, Maximum=<0 0.6 0>)
     7=&RandomRangeVector(Minimum=<0 0.1 0>, Maximum=<0 0.4 0>)
     Index=&SpawnGroup()
     Num=::NewEnumerator6
    )
    Use Rotation Vector=true
   )
   &CameraOffset(Camera Offset Amount=10)
  ]
  particleUpdateScript [
   &ParticleState()
   &ScaleMeshSize(Scale Factor=&VectorFromFloat(Value=&MergeHelper_FloatSwitch(
      0=1
      1=1
      2=&MakeCustomFloatFromBool(Bool=$User.bIsFootHidden, True Float=0, False Float=1)
      3=&MakeCustomFloatFromBool(Bool=$User.bIsFootHidden, True Float=0, False Float=1)
      4=&MakeCustomFloatFromBool(Bool=$User.bIsFootHidden, True Float=0, False Float=1)
      5=&MakeCustomFloatFromBool(Bool=$User.bIsFootHidden, True Float=0, False Float=1)
      6=&MakeCustomFloatFromBool(Bool=$User.bIsFootHidden, True Float=0, False Float=1)
      7=&MakeCustomFloatFromBool(Bool=$User.bIsFootHidden, True Float=0, False Float=1)
      Index=&SpawnGroup()
      Num=::NewEnumerator6
     )))
   &ScaleColor(
    Scale Alpha=&MergeHelper_FloatSwitch(
     0=1
     1=1
     2=&MakeCustomFloatFromBool(Bool=$User.bIsFootHidden, True Float=0, False Float=1)
     3=&MakeCustomFloatFromBool(Bool=$User.bIsFootHidden, True Float=0, False Float=1)
     4=&MakeCustomFloatFromBool(Bool=$User.bIsFootHidden, True Float=0, False Float=1)
     5=&MakeCustomFloatFromBool(Bool=$User.bIsFootHidden, True Float=0, False Float=1)
     6=&MakeCustomFloatFromBool(Bool=$User.bIsFootHidden, True Float=0, False Float=1)
     7=&MakeCustomFloatFromBool(Bool=$User.bIsFootHidden, True Float=0, False Float=1)
     Index=&SpawnGroup()
     Num=::NewEnumerator6
    )
    ScaleA=true
    ScaleRGB=true
   )
   &DynamicMaterialParameters(
    Index 0 Param 1=&FloatFromCurve(FloatCurve=DataInterface() {propertyValues {Curve {keys [{tangentMode=RCTM_User, time=0, value=0}, <at 0.333 0.5>, {tangentMode=RCTM_User, time=1, value=1}]}}})
    Index 0 Param 2=&FloatFromCurve(
     FloatCurve=DataInterface() {propertyValues {Curve {keys [<at 0.333 1>,<at 0.666 0.333>]}}}
     Scale Curve=&RandomRangeFloat(Minimum=0.6, Maximum=1)
    )
    Index 0 Param 3=0.75
    Index 0 Param 4=&FloatFromCurve(
     FloatCurve=DataInterface() {propertyValues {Curve {keys [
         {interpMode=RCIM_Cubic, tangentMode=RCTM_User, time=0, value=0}
         {interpMode=RCIM_Cubic, tangentMode=RCTM_User, time=0.5, value=1}
         {tangentMode=RCTM_User, time=1, value=0}
        ]}}}
     Scale Curve=33
    )
    Index 1 Param 1=1
    Module.Index0Param0Write=true
    Module.Index0Param1Write=true
    Module.Index0Param2Write=true
    Module.Index0Param3Write=true
    Module.Index1Param0Write=true
    Module.Index1Param1Write=false
    Module.Index1Param2Write=false
    Module.Index1Param3Write=false
    Write Parameter Index 1=true
   )
   /SetVariables(Particles.MeshIndex=&RandomRangeInt(Minimum=0, Maximum=4)) {isSetParameters=true, module=SetVariables_#3}
   &SocketLocation(
    SamplePosOffset=true
    Sampled Position Offset=&MergeHelper_VectorSwitch(
     0=&RandomRangeVector(Minimum=<0 0 0>, Maximum=<25 0 0>)
     1=&RandomRangeVector(Minimum=<0 0 0>, Maximum=<25 0 0>)
     2=&RandomRangeVector(Minimum=<-12 0 0>, Maximum=<25 0 0>)
     3=&RandomRangeVector(Minimum=<-12 0 0>, Maximum=<25 0 0>)
     4=&RandomRangeVector(Minimum=<10 -1 -1>, Maximum=<11 1 1>)
     5=&RandomRangeVector(Minimum=<10 -1 -1>, Maximum=<11 1 1>)
     6=&RandomRangeVector(Minimum=<10 -1 -1>, Maximum=<11 1 1>)
     7=&RandomRangeVector(Minimum=<10 -1 -1>, Maximum=<11 1 1>)
     Index=&SpawnGroup()
     Num=::NewEnumerator6
    )
    Socket Index=&Add_Integer(A=10, B=&SpawnGroup())
    Socket Reader=$System.Socket Reader
    Socket Sampling Mode=::NewEnumerator3
    VelocitySpeedLimit=false
   )
   &FN_Sprite_Attributes(
    Camera Offset=15
    Color=&MakeLinearColorFromVectorAndFloat(
     "Vector (RGB)"=&Multiply_Vector(A=&MakeVectorFromLinearColorRGB(Color=<rgba 1 0.6535 0.0208 1>), B=<1 1 1>)
     "Float (Alpha)"=&Multiply_Float(
      A=&FloatFromCurve(
       FloatCurve=DataInterface() {propertyValues {Curve {keys [
           {interpMode=RCIM_Cubic, tangentMode=RCTM_User, time=0, value=0}
           {interpMode=RCIM_Cubic, tangentMode=RCTM_User, time=0.3, value=1}
           {tangentMode=RCTM_User, time=1, value=0}
          ]}}}
       Scale Curve=0.7
      )
      B=1
     )
    )
    Create Camera Offset Attribute=true
    Create Color Attribute=true
    Create SpriteSize Attribute=true
    SpriteSize=&Multiply_Vector2D_ByFloat(
     Float=&FloatFromCurve(
      FloatCurve=DataInterface() {propertyValues {Curve {keys [
          {interpMode=RCIM_Cubic, tangentMode=RCTM_User, time=0, value=0}
          {interpMode=RCIM_Cubic, tangentMode=RCTM_User, time=0.3, value=1}
          {tangentMode=RCTM_User, time=1, value=0}
         ]}}}
      Scale Curve=&MergeHelper_FloatSwitch(
       0=1
       1=1
       2=&MakeCustomFloatFromBool(Bool=$User.bIsFootHidden, True Float=0, False Float=1)
       3=&MakeCustomFloatFromBool(Bool=$User.bIsFootHidden, True Float=0, False Float=1)
       4=&MakeCustomFloatFromBool(Bool=$User.bIsFootHidden, True Float=0, False Float=1)
       5=&MakeCustomFloatFromBool(Bool=$User.bIsFootHidden, True Float=0, False Float=1)
       6=&MakeCustomFloatFromBool(Bool=$User.bIsFootHidden, True Float=0, False Float=1)
       7=&MakeCustomFloatFromBool(Bool=$User.bIsFootHidden, True Float=0, False Float=1)
       Index=&SpawnGroup()
       Num=::NewEnumerator6
      )
     )
     Vector2D=<25 25>
    )
    Write to Module Namespace=true
   )
  ]
  rendererClasses [&NiagaraMeshRendererProperties, &NiagaraSpriteRendererProperties]
  renderers [
   {rendererClass=&NiagaraMeshRendererProperties, rendererIndex=0}
   {rendererClass=&NiagaraMeshRendererProperties, rendererIndex=1}
   {rendererClass=&NiagaraMeshRendererProperties, rendererIndex=2}
   {rendererClass=&NiagaraMeshRendererProperties, rendererIndex=3}
   {rendererClass=&NiagaraSpriteRendererProperties, rendererIndex=4}
  ]
 }
}
summary {
 systemName=NS_Character_ScoopPluck_Idle
 emitters [
  {emitterName=EyesSwooshesGlow, rendererClasses [&NiagaraSpriteRendererProperties]}
  {emitterName=LightningBody, rendererClasses [&NiagaraMeshRendererProperties, &NiagaraSpriteRendererProperties]}
  {emitterName=LightningLegs, rendererClasses [&NiagaraMeshRendererProperties, &NiagaraSpriteRendererProperties]}
 ]
 userVariables [
  User.bIsADS {type=NiagaraBool}
  User.bIsFootHidden {type=NiagaraBool}
  User.bIsFrontEnd {type=NiagaraBool}
  User.bIsFrontEndPreview {type=NiagaraBool}
  User.bIsHandHidden {type=NiagaraBool}
  User.bIsHeadHidden {type=NiagaraBool}
  User.Helmet {type=NiagaraFloat, defaultValue=0, description=""}
  User.KillCount {type=NiagaraInt32}
  User.Variant {type=NiagaraFloat, defaultValue=0, description=""}
 ]
}
