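# Parent character material for the ScoopPluck skin that adds a timed reactive emissive glow, between the shared 2023 PreFX and PostFX functions.
# Time since Timestamp over a duration shapes a rise-and-fall pulse with separate begin and end exponents; a debug switch loops it.
# A texture or FX mask, inverse Fresnel and color shape the glow, with optional ear tint split by pre-skinned X position.
# An optional advanced clear coat precedes the glow.
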Material /BRCosmetics/Characters/Player/Male/Medium/Bodies/M_MED_ScoopPluck/Materials/M_ScoopPluck_Reactivity_Parent
graph {
 properties {
  bUseMaterialAttributes=true
  bUsedWithClothing=true
  bUsedWithMeshDeformer=true
  bUsedWithMorphTargets=true
  bUsedWithSkeletalMesh=true
  pixelDepthOffsetMode=PDOM_Legacy
  shadingModel=MSM_FromMaterialExpression
  usedShadingModels=Default Lit | Subsurface | Clear Coat | Subsurface Profile | Cloth
 }
 nodes [
  n1 = MaterialFunctionCall() {materialFunction=&MF_CharacterParent_2023_PreFX}
  n2 = MaterialFunctionCall!(Mask=@n1.Tech Art Mask, Fresnel=@n1.Fresnel, In=@n1.Result) {materialFunction=/Game/Global/MaterialFunctions/TechArt/MF_Advanced_CarClearCoat_v2}
  n3 = GetMaterialAttributes(None=StaticSwitchParameter(
    True=SetMaterialAttributes!(MaterialAttributes=@n2.Result, ClearCoatBottomNormal=@n2.Result) {attributeSetTypes [#1]}
    False=@n1.Result
   ) {parameterName=UseAdvancedClearCoat}) {attributeGetTypes [#2]}
  n4 = TextureSampleParameter2D() {
   group=Reactivity
   parameterName=ReactivityMask
   samplerType=SAMPLERTYPE_Masks
   texture=/BRCosmetics/Characters/Player/Male/Medium/Bodies/M_MED_ScoopPluck/Textures/T_ScoopPluck_FaceAcc_VisorMask
  }
  n5 = VectorParameter() {group=Reactivity, parameterName=ReactivityColor, defaultValue=<rgba 2 34.6667 100 1>}
  n6 = VectorParameter() {group=Reactivity, parameterName=EarsColor, defaultValue=<rgba 2 34.6667 100 1>}
  n7 = LocalPosition() {localOrigin=InstancePreSkinning}
  n8 = Abs(None=ComponentMask(None=VertexInterpolator(VS=@n7.XYZ).PS) {r=true})
  n9 = ScalarParameter() {parameterName=Range}
  n10 = Round(None=Saturate(None=Subtract(A=@n8, B=@n9)))
  n11 = Multiply(
   A=StaticSwitchParameter(True=ChannelMaskParameter(None=@n1.Tech Art Mask) {group=Reactivity, parameterName=ReactivityMaskChannel}.R, False=@n4.R) {parameterName=UseFXMask}
   B=LinearInterpolate(A=@n5.RGB, B=LinearInterpolate(A=@n5.RGB, B=@n6.RGB, Alpha=@n10), Alpha=ScalarParameter() {parameterName=SeparateEars})
  )
  n12 = Divide(
   A=Subtract(A=Time(), B=ScalarParameter() {defaultValue=-10000, group=Reactivity, parameterName=Timestamp})
   B=ScalarParameter() {defaultValue=4, group=Reactivity, parameterName=ReactivityDuration}
  )
  n13 = Fmod(A=@n12, B=Constant() {r=2})
  n14 = StaticSwitchParameter(True=Multiply(A=Frac(None=@n12), B=Floor(None=@n13)), False=@n12) {group=Reactivity, parameterName=DebugReactivity}
  n15 = ConstantBiasScale(None=@n14) {bias=-0.5, scale=2}
  n16 = Power(
   Base=OneMinus(None=Abs(None=@n15))
   Exp=LinearInterpolate(
    A=ScalarParameter() {defaultValue=0.5, group=Reactivity, parameterName=ReactivityPowerBegin}
    B=ScalarParameter() {defaultValue=2, group=Reactivity, parameterName=ReactivityPowerEnd}
    Alpha=Round(None=@n14)
   )
  )
  n17 = Fresnel(ExponentIn=ScalarParameter() {defaultValue=0.5, group=Reactivity, parameterName=Reactivity Fresnel Exp})
  n18 = Multiply(A=Multiply(A=Multiply(A=@n11, B=@n16), B=OneMinus(None=@n17)), B=ScalarParameter() {defaultValue=1, parameterName=ReactivityEnabled})
  MaterialFunctionCall(
   Material=SetMaterialAttributes!(MaterialAttributes=@n3.MaterialAttributes, Emissive Color=Add(A=@n3.Emissive Color, B=@n18)) {attributeSetTypes [#2]}
   M=@n1.M
   Pre Skinned Local Position=@n1.M
   Distance Blend=@n1.M
   Fresnel=@n1.M
  ) {materialFunction=&MF_CharacterParent_2023_PostFX}
 ]
}
